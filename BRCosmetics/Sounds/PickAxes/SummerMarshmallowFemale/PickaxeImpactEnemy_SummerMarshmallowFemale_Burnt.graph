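# Sound cue for the burnt variant of the SummerMarshmallowFemale pickaxe enemy impact, drawing on "BurntAdd" waves.
# A short fixed delay feeds a Modulator, then a Random node picks one of five equally weighted waves.
# It uses the shared player pickaxe impact attenuation, the cosmetic pickaxe impact property sheet and the Fort_Impacts sound class, at reduced volume.
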SoundCue /BRCosmetics/Sounds/PickAxes/SummerMarshmallowFemale/PickaxeImpactEnemy_SummerMarshmallowFemale_Burnt
root=Delay_0
nodes [
 {id=Delay_0, children [Modulator_0], properties {delayMax=0.1, delayMin=0.1}}
 {id=Modulator_0, children [Random_0]}
 {id=Random_0, children [WavePlayer_0,WavePlayer_1,WavePlayer_2,WavePlayer_3,WavePlayer_4], properties {weights [1,1,1,1,1]}}
 {
  id=WavePlayer_0
  properties {soundWaveAssetPtr=&/Pickaxe_SummerMarshmallowFemale_Impact_BurntAdd_01}
 }
 {
  id=WavePlayer_1
  properties {soundWaveAssetPtr=&/Pickaxe_SummerMarshmallowFemale_Impact_BurntAdd_02}
 }
 {
  id=WavePlayer_2
  properties {soundWaveAssetPtr=&/Pickaxe_SummerMarshmallowFemale_Impact_BurntAdd_03}
 }
 {
  id=WavePlayer_3
  properties {soundWaveAssetPtr=&/Pickaxe_SummerMarshmallowFemale_Impact_BurntAdd_04}
 }
 {
  id=WavePlayer_4
  properties {soundWaveAssetPtr=&/Pickaxe_SummerMarshmallowFemale_Impact_BurntAdd_05}
 }
]
properties {
 attenuationSettings=/Game/Sounds/Attenuation/Fort_Weapon_Pickaxe_Player_Impact_Att
 audioPropertiesSheet=/BRCosmetics/Sounds/Settings/PropertySheets/Pickaxe/APS_Cosmetic_Pickaxe_Impact
 maxDistance=3512
 soundClassObject=&Fort_Impacts
 volumeMultiplier=0.5625
}
